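# Subsurface material for instanced jungle grind rail boosters, using BoostShroom textures with world-noise hue variation and pulsing, animated arrow glow.
# Per-instance custom data supplies booster count, index, per-side state, timing, direction and trail length.
# A time-driven wave compares each instance's index to a moving position, producing a falloff that drives emissive and world position offset.
# Each side's state sign selects arrow channel and color, while time of day scales arrow intensity.
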Material /ChronoEnviroItems/JungleGrindRails/TechArt/Materials/M_Chrono_Jungle_GrindRail_Booster
graph {
 properties {bUsedWithInstancedStaticMeshes=true, pixelDepthOffsetMode=PDOM_Legacy, shadingModel=MSM_Subsurface, usedShadingModels=Subsurface}
 nodes [
  n1 = Constant()
  n2 = ScalarParameter() {defaultValue=1, parameterName=Animation Interval time}
  n3 = Add(A=Multiply(A=NamedRerouteUsage()) {constB=-1}, B=@n2)
  n4 = Add(A=Subtract(A=NamedRerouteUsage(), B=@n2), B=NamedRerouteUsage())
  n5 = NamedRerouteUsage()
  n6 = Distance(A=NamedRerouteUsage(), B=Round(None=LinearInterpolate(A=@n3, B=@n4, Alpha=@n5) {constB=20}))
  n7 = Subtract(A=NamedRerouteUsage(), B=@n6)
  n8 = NamedRerouteUsage()
  n9 = If(A=@n6, B=NamedRerouteUsage(), "A > B"=@n1, "A == B"=@n1, "A < B"=Power(Base=Divide(A=@n7, B=@n8)))
  n10 = Multiply(A=NamedRerouteUsage(), B=LinearInterpolate(A=ScalarParameter() {parameterName=Arrow Default intensity}, B=NamedRerouteUsage(), Alpha=@n9))
  Add(A=@n10)
  Constant()
  n11 = WorldPosition() {worldPositionShaderOffset=WPT_ExcludeAllShaderOffsets}
  n12 = ScalarParameter() {defaultValue=300, parameterName=BoosterColorRandomScale}
  n13 = ConstantBiasScale(None=TextureSample(UVs=Divide(A=@n11.XY, B=@n12) {constB=500}) {samplerType=SAMPLERTYPE_Grayscale, texture=&LowResBlurredNoise}.RGB) {bias=-0.5, scale=2}
  n14 = TextureSampleParameter2D() {parameterName=Diffuse, texture=/ChronoEnviroItems/JungleGrindRails/Textures/T_BoostShroom_B_D}
  MaterialFunctionCall(Hue Shift Percentage=Multiply(A=@n13, B=ScalarParameter() {defaultValue=0.1, parameterName=BoosterColorRandomAmount}), Texture=@n14.RGB) {materialFunction=&HueShift}
  MaterialFunctionCall() {materialFunction=&DebugTimeSine}
  Multiply(A=TextureSampleParameter2D() {parameterName=Spec, samplerType=SAMPLERTYPE_Masks, texture=/ChronoEnviroItems/JungleGrindRails/Textures/T_BoostShroom_A_S}.G) {constB=0.8}
  n15 = Constant() {r=0.5}
  Multiply(A=MaterialFunctionCall(Hue Shift Percentage=Multiply(A=@n13, B=@n15), Texture=@n14.RGB) {materialFunction=&HueShift}.Result) {constB=0.5}
  Multiply(A=@n14.RGB, B=VectorParameter() {parameterName=SSC, defaultValue=<rgba 1 1 1 0.6>}.RGB)
  n16 = Multiply(A=@n9, B=ScalarParameter() {defaultValue=25, parameterName=BoosterPulsingWPO}) {constB=10}
  n17 = Constant3Vector() {constant=<rgba 0 0 1 0>}
  n18 = WorldPosition()
  n19 = Constant3Vector()
  n20 = ComponentMask(None=Subtract(A=@n18.XYZ, B=TransformPosition(None=@n19) {transformSourceType=TRANSFORMPOSSOURCE_Instance, transformType=TRANSFORMPOSSOURCE_World})) {r=true}
  n21 = ScalarParameter() {defaultValue=0.03, parameterName=BoosterPulsingWPO_XY}
  n22 = Add(
   A=Multiply(A=Transform(None=@n17) {transformSourceType=TRANSFORMSOURCE_Instance})
   B=AppendVector(A=Multiply(A=@n20, B=@n21), B=Constant2Vector())
  )
  n23 = SphereMask(
   A=@n18.XYZ
   B=TransformPosition(None=Constant3Vector()) {transformSourceType=TRANSFORMPOSSOURCE_Instance, transformType=TRANSFORMPOSSOURCE_World}
  ) {attenuationRadius=40, hardnessPercent=0}
  n24 = Power(Base=OneMinus(None=Saturate(None=@n23))) {constExponent=4}
  n25 = ScalarParameter() {defaultValue=1, parameterName=BoosterPulsingWPO_Z}
  NamedRerouteDeclaration(None=Multiply(A=Multiply(A=@n16, B=@n22), B=Multiply(A=@n24, B=@n25))) {name=WPO, nodeColor=<rgba 0.5469 1 0 1>}
  n26 = PerInstanceCustomData!(DefaultValue) {dataIndex=2}
  n27 = VectorParameter() {parameterName=BoosterEmissive, defaultValue=<rgba 0 1.3148 2 1>}
  n28 = ScalarParameter() {defaultValue=1, parameterName=BoosterEmissiveMult}
  n29 = Constant3Vector() {constant=<rgba 1 0.05 0.05 0>}
  NamedRerouteDeclaration(None=If(A=VertexInterpolator(VS=@n26).PS, "A > B"=Multiply(A=@n27.RGB, B=@n28), "A == B"=Constant3Vector(), "A < B"=@n29)) {name=SideA Color, nodeColor=<rgba 0 0.2969 1 1>}
  n30 = PerInstanceCustomData() {dataIndex=3}
  NamedRerouteDeclaration(None=If(A=VertexInterpolator(VS=@n30).PS, "A > B"=@n27.RGB, "A == B"=Constant3Vector(), "A < B"=@n29)) {name=SideB_color, nodeColor=<rgba 0.2188 1 0 1>}
  n31 = PerInstanceCustomData() {dataIndex=2}
  n32 = VertexInterpolator(VS=@n31)
  n33 = TextureSampleParameter2D() {
   parameterName=T_Arrow Emissive_E
   samplerSource=SSM_Wrap_WorldGroupSettings
   texture=/ChronoEnviroItems/JungleGrindRails/Textures/T_BoostShroom_A_E
  }
  n34 = TextureSampleParameter2D() {
   parameterName=T_Arrow Emissive_E
   samplerSource=SSM_Wrap_WorldGroupSettings
   texture=&T_BoostShroom_A_E
  }
  NamedRerouteDeclaration(None=If(
    A=ShaderStageSwitch(PixelShader=@n32.PS, VertexShader=@n31)
    "A > B"=Power(Base=@n33.A) {constExponent=4}
    "A == B"=Constant3Vector()
    "A < B"=Power(Base=@n34.R) {constExponent=4}
   )) {name=SideA Arrow, nodeColor=<rgba 0 0.2969 1 1>}
  n35 = PerInstanceCustomData() {dataIndex=3}
  n36 = TextureSampleParameter2D() {
   parameterName=T_Arrow Emissive_E
   samplerSource=SSM_Wrap_WorldGroupSettings
   texture=&T_BoostShroom_A_E
  }
  n37 = TextureSampleParameter2D() {
   parameterName=T_Arrow Emissive_E
   samplerSource=SSM_Wrap_WorldGroupSettings
   texture=&T_BoostShroom_A_E
  }
  NamedRerouteDeclaration(None=If(
    A=VertexInterpolator(VS=@n35).PS
    "A > B"=Power(Base=@n36.A) {constExponent=4}
    "A == B"=Constant3Vector()
    "A < B"=Power(Base=@n37.R) {constExponent=4}
   )) {name=SideB Arrow, nodeColor=<rgba 0 0.2969 1 1>}
  n38 = Time()
  n39 = ScalarParameter() {defaultValue=0.25, parameterName=Param_2}
  n40 = Multiply(A=Sine(None=Multiply(A=@n38, B=@n39)), B=ScalarParameter() {defaultValue=0.76})
  n41 = PerInstanceCustomData() {dataIndex=4}
  n42 = Multiply(
   A=ShaderStageSwitch(PixelShader=VertexInterpolator(VS=@n41).PS, VertexShader=@n41)
   B=ScalarParameter() {defaultValue=2.534, parameterName=Param_1}
  ) {constB=10}
  n43 = MaterialFunctionCall(Value=@n40, Period=@n42, -1 to 1=StaticBool()) {materialFunction=&LinearSine}
  n44 = MaterialFunctionCall(Value=Add(A=@n40, B=Multiply(A=@n42) {constB=0.5}), Period=@n42, -1 to 1=StaticBool()) {materialFunction=&LinearSine}
  n45 = Max(A=Multiply(A=@n43.Linear Sine, B=@n43.Linear Sine), B=Multiply(A=@n44.Linear Sine, B=@n44.Linear Sine))
  n46 = OneMinus(None=@n45)
  n47 = PerInstanceCustomData() {dataIndex=5}
  NamedRerouteDeclaration(None=Saturate(None=LinearInterpolate(A=@n45, B=@n46, Alpha=@n47))) {name=SideA_animation, nodeColor=<rgba 0 0.625 1 1>}
  n48 = MaterialFunctionCall(Value=Time(), Period=@n42, -1 to 1=StaticBool()) {materialFunction=&LinearSine}
  n49 = MaterialFunctionCall(Value=Add(A=Time(), B=Multiply(A=@n42) {constB=0.5}), Period=@n42, -1 to 1=StaticBool()) {materialFunction=&LinearSine}
  n50 = Max(A=Multiply(A=@n48.Linear Sine, B=@n48.Linear Sine), B=Multiply(A=@n49.Linear Sine, B=@n49.Linear Sine))
  n51 = OneMinus(None=@n50)
  n52 = PerInstanceCustomData() {dataIndex=6}
  NamedRerouteDeclaration(None=Saturate(None=LinearInterpolate(A=@n50, B=@n51, Alpha=@n52))) {name=SideB_animation, nodeColor=<rgba 0 0.625 1 1>}
  n53 = PerInstanceCustomData() {constDefaultValue=3, dataIndex=7}
  n54 = VertexInterpolator(VS=@n53)
  NamedRerouteDeclaration(None=Multiply(A=ShaderStageSwitch(PixelShader=@n54.PS, VertexShader=@n53)) {constB=2}) {name=Animation trailing length, nodeColor=<rgba 0.7578 1 0 1>}
  n55 = PerInstanceCustomData()
  NamedRerouteDeclaration(None=ShaderStageSwitch(PixelShader=VertexInterpolator(VS=@n55).PS, VertexShader=@n55)) {name=Total number of booster models, nodeColor=<rgba 0.5938 0 1 1>}
  n56 = PerInstanceCustomData() {dataIndex=1}
  NamedRerouteDeclaration(None=ShaderStageSwitch(PixelShader=VertexInterpolator(VS=@n56).PS, VertexShader=@n56)) {name=Instance staic mesh index, nodeColor=<rgba 0 0.4609 1 1>}
  NamedRerouteDeclaration(None=MaterialFunctionCall(
    Value_Morning=ScalarParameter() {defaultValue=3, group=Emissive, parameterName=Arrow Emissive intensity Morning}
    Value_Day=ScalarParameter() {defaultValue=15, parameterName=Arrow Emissive intensity Day}
    Value_Evening=ScalarParameter() {defaultValue=3, parameterName=Arrow Emissive intensity Evening}
    Value_Night=ScalarParameter() {defaultValue=3, parameterName=Arrow Emissive intensity Night}
   ) {materialFunction=&Scalar_Time_of_Day}.Result) {name=The Emissive intensity of the arrow, nodeColor=<rgba 0 1 0.4609 1>}
  n57 = PerInstanceCustomData() {dataIndex=5}
  NamedRerouteDeclaration(None=ShaderStageSwitch(PixelShader=VertexInterpolator(VS=@n57).PS, VertexShader=@n57)) {name=SideA_Reverse animation direction, nodeColor=<rgba 1 0 0.7734 1>}
  n58 = PerInstanceCustomData() {dataIndex=6}
  NamedRerouteDeclaration(None=ShaderStageSwitch(PixelShader=VertexInterpolator(VS=@n58).PS, VertexShader=@n58)) {name=SideB_Reverse animation direction, nodeColor=<rgba 1 0 0.8906 1>}
  n59 = NamedRerouteUsage()
  n60 = Fresnel()
  n61 = TextureSampleParameter2D() {parameterName=Diffuse, texture=&T_BoostShroom_B_D}
  n62 = Power(Base=OneMinus(None=Multiply(A=@n61.R))) {constExponent=10}
  NamedRerouteDeclaration(None=Multiply(A=Multiply(A=@n59, B=@n10), B=Multiply(A=@n60, B=@n62))) {name=Emissive, nodeColor=<rgba 1 0.7969 0 1>}
  NamedRerouteUsage()
  NamedRerouteUsage()
  NamedRerouteUsage()
  TextureSampleParameter2D() {parameterName=Normal, samplerType=SAMPLERTYPE_Normal, texture=/ChronoEnviroItems/JungleGrindRails/Textures/T_BoostShroom_A_N}
  Time()
 ]
}
