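# Parent material for jungle grind rail spline meshes, with moss overgrowth and gameplay-reactive glow.
# A height-blended, noise-broken moss mask layers mossy ground over rail textures; the Is3xGrindrail switch changes texture sources and moss masking.
# Custom primitive data drives emissive: a player-position sphere mask responding to sprint and jump-off, idle and distance glow, and optional end glows.
# Emissive varies by time of day and compensates for eye adaptation off mobile; world noise also builds vertex offset.
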Material /ChronoEnviroItems/JungleGrindRails/TechArt/Materials/M_Chrono_Jungle_GrindRail_Parent
graph {
 properties {bUsedWithSplineMeshes=true, pixelDepthOffsetMode=PDOM_Legacy, usedShadingModels=Default Lit}
 nodes [
  Constant3Vector() {constant=<rgba 0 0.0134 0.0156 0>}
  Constant3Vector() {constant=<rgba 0 0 1 0>}
  n1 = MaterialFunctionCall(Float2=TextureCoordinate()) {materialFunction=&BreakOutFloat2Components}
  ConstantBiasScale(None=@n1.R) {bias=-0.5, scale=2}
  Constant()
  n2 = TextureSample(UVs=NamedRerouteUsage()) {samplerType=SAMPLERTYPE_Normal, texture=/Game/Environments/Asteria/Props/GrindRailVines/GrindRailVines_3D_N}
  n3 = TextureSample(UVs=NamedRerouteUsage()) {samplerType=SAMPLERTYPE_Normal, texture=/Game/Environments/Asteria/Sets/ZenGarden/Textures/T_Asteria_Zen_MossyGround_01_N}
  n4 = NamedRerouteUsage()
  n5 = LinearInterpolate(A=@n2.RGB, B=@n3.RGB, Alpha=@n4)
  MaterialFunctionCall(BaseNormal=@n2.RGB, AdditionalNormal=@n5) {materialFunction=&BlendAngleCorrectedNormals}
  n6 = ScalarParameter() {defaultValue=1, parameterName=Tiling_V_MossMult, primitiveDataIndex=18}
  Multiply(A=ScalarParameter() {bUseCustomPrimitiveData=true, defaultValue=1, parameterName=Tiling_V, primitiveDataIndex=18}, B=@n6)
  n7 = NamedRerouteUsage()
  Multiply(A=@n7, B=ScalarParameter() {defaultValue=0.1, parameterName=MossHeightTextureTiling}) {constB=0.1}
  n8 = WorldPosition()
  n9 = CollectionParameter() {collection=&FortniteMaterialParameters, parameterName=PlayerPosition}
  n10 = ComponentMask(None=@n9) {b=true, g=true, r=true}
  n11 = LinearInterpolate(
   A=LinearInterpolate(
    A=ScalarParameter() {defaultValue=150, group=EmissiveColor, parameterName=RadiusDefaultEffect}
    B=ScalarParameter() {defaultValue=250, group=EmissiveColor, parameterName=RadiusSprintingEffectRadius}
    Alpha=ScalarParameter() {bUseCustomPrimitiveData=true, group=Gameplay, parameterName=sprinting state, primitiveDataIndex=8}
   )
   Alpha=ScalarParameter() {bUseCustomPrimitiveData=true, defaultValue=1, parameterName=IdleTail light switch, primitiveDataIndex=6}
  ) {constB=0}
  n12 = ScalarParameter() {group=EmissiveColor, parameterName=HardnessRadius}
  n13 = Multiply(A=Power(Base=SphereMask(A=@n8.XYZ, B=@n10, Radius=@n11, Hardness=@n12)), B=Constant() {r=1.02})
  Multiply(
   A=LinearInterpolate(A=@n13, Alpha=ScalarParameter() {bUseCustomPrimitiveData=true, group=Gameplay, parameterName=Jumpoff, primitiveDataIndex=7}) {constB=0}
   B=LinearInterpolate(Alpha=ScalarParameter() {bUseCustomPrimitiveData=true, group=Gameplay, parameterName=sprinting state, primitiveDataIndex=8}) {constA=0.7}
  )
  n14 = ScalarParameter() {defaultValue=0.3, parameterName=Tiling_V_Mult, primitiveDataIndex=18}
  Multiply(A=ScalarParameter() {bUseCustomPrimitiveData=true, defaultValue=1, parameterName=Tiling_V, primitiveDataIndex=18}, B=@n14)
  n15 = MaterialFunctionCall(Float2=TextureCoordinate()) {materialFunction=&BreakOutFloat2Components}
  n16 = ScalarParameter() {defaultValue=1, parameterName=Tiling_U}
  NamedRerouteDeclaration(None=AppendVector(A=Multiply(A=@n15.R, B=@n16), B=Multiply(A=@n15.G, B=@n14))) {name=TexCocord, nodeColor=<rgba 1 0 0.0703 1>}
  n17 = WorldPosition() {worldPositionShaderOffset=WPT_ExcludeAllShaderOffsets}
  n18 = ScalarParameter() {defaultValue=300, parameterName=WorldNoise_Size}
  n19 = TextureSampleParameter2D(UVs=ComponentMask(None=Divide(A=@n17.XYZ, B=@n18)) {b=true, g=true}) {parameterName=WorldNoise, samplerType=SAMPLERTYPE_LinearGrayscale, texture=/Game/Global/Textures/Default/Noise/T_SwirlNoise_N_height}
  n20 = Multiply(
   A=ConstantBiasScale(None=@n19.R) {bias=-0.5, scale=2}
   B=Multiply(A=@n19.R, B=ScalarParameter() {defaultValue=15, parameterName=WorldNoise_ThicknessOffset}) {constB=0.02}
  ) {constB=0.02}
  NamedRerouteDeclaration(None=LinearInterpolate(
    A=MaterialFunctionCall(X=@n20, Y=@n20, Z=@n20) {materialFunction=&MakeFloat3}.Result
    Alpha=NamedRerouteUsage()
   ) {constAlpha=1, constB=0}) {name=WPO, nodeColor=<rgba 0 1 0.3438 1>}
  n21 = ScalarParameter() {defaultValue=8000, group=EmissiveColor, parameterName=BlendDistance}
  n22 = Multiply(A=@n21) {constB=1.5}
  n23 = ScalarParameter() {bUseCustomPrimitiveData=true, defaultValue=1, parameterName=IdleTail light switch, primitiveDataIndex=6}
  n24 = Clamp(None=Divide(A=PixelDepth().R, B=LinearInterpolate(A=@n22, B=@n21, Alpha=@n23)) {constB=5000})
  n25 = Constant() {r=2}
  NamedRerouteDeclaration(None=Saturate(None=Power(Base=@n24, Exp=@n25) {constExponent=1.2})) {name=PixelDepthMask, nodeColor=<rgba 0 0.5781 1 1>}
  n26 = MaterialFunctionCall(Float2=TextureCoordinate()) {materialFunction=&BreakOutFloat2Components}
  n27 = AppendVector(A=Multiply(A=@n26.R, B=ScalarParameter() {defaultValue=1, parameterName=Tiling_U_Moss}), B=Multiply(A=@n26.G, B=@n6))
  NamedRerouteDeclaration(None=@n27) {name=Texcord_Moss, nodeColor=<rgba 0 1 0.0625 1>}
  n28 = VectorParameter() {bUseCustomPrimitiveData=true, group=EmissiveColor, parameterName=EmissiveColorLight, primitiveDataIndex=2, defaultValue=<rgba 0 2 1 1>}
  n29 = Saturate(None=InverseLinearInterpolate(
    A=ScalarParameter() {defaultValue=100000, group=EmissiveColor, parameterName=GlowAtDistance_EndDistance}
    B=ScalarParameter() {defaultValue=2500, group=EmissiveColor, parameterName=GlowAtDistance_StartDistance}
    Value=PixelDepth().R
   ))
  n30 = Multiply(A=@n28.RGB, B=@n29)
  n31 = Add(A=@n30, B=Multiply(A=@n30, B=@n29))
  n32 = ScalarParameter() {parameterName=GlowAtDistanceDesaturation}
  NamedRerouteDeclaration(None=Multiply(
    A=Desaturation(None=@n31, Fraction=@n32)
    B=ScalarParameter() {defaultValue=5, group=EmissiveColor, parameterName=Glow indensity at a distance}
   )) {name=Glow at a distance, nodeColor=<rgba 0 1 0.0391 1>}
  n33 = TextureSample(UVs=NamedRerouteUsage()) {samplerType=SAMPLERTYPE_Normal, texture=/ChronoEnviroItems/JungleGrindRails/Textures/T_JungleGrindRail_A_N}
  NamedRerouteDeclaration(None=StaticSwitchParameter(True=@n5, False=LinearInterpolate(A=@n33.RGB, B=@n3.RGB, Alpha=@n4)) {parameterName=Is3xGrindrail}) {name=Normal, nodeColor=<rgba 0 0.7187 1 1>}
  n34 = NamedRerouteUsage()
  n35 = Saturate(None=LinearInterpolate(
    A=ScalarParameter() {defaultValue=1.0992, parameterName=Roughness}
    B=ScalarParameter() {defaultValue=-2.0378, parameterName=Roughness2}
    Alpha=@n34
   ))
  n36 = TextureSample(UVs=NamedRerouteUsage()) {samplerType=SAMPLERTYPE_Masks, texture=/ChronoEnviroItems/JungleGrindRails/Textures/T_JungleGrindRail_A_S}
  NamedRerouteDeclaration(None=LinearInterpolate(
    A=StaticSwitchParameter(True=@n35, False=@n36.B) {parameterName=Is3xGrindrail}
    B=ScalarParameter() {defaultValue=1, parameterName=Moss_Rough, primitiveDataIndex=18}
    Alpha=NamedRerouteUsage()
   )) {name=Rough, nodeColor=<rgba 0 0.6953 1 1>}
  n37 = Saturate(None=LinearInterpolate(
    A=ScalarParameter() {defaultValue=0.068, parameterName=Specular}
    B=ScalarParameter() {defaultValue=0.7685, parameterName=Specular2}
    Alpha=@n34
   ))
  NamedRerouteDeclaration(None=LinearInterpolate(
    A=StaticSwitchParameter(True=@n37, False=@n36.R) {parameterName=Is3xGrindrail}
    B=ScalarParameter() {parameterName=Moss_Spec, primitiveDataIndex=18}
    Alpha=NamedRerouteUsage()
   )) {name=Spec, nodeColor=<rgba 0.0312 1 0 1>}
  n38 = Multiply(A=TextureCoordinate(), B=AppendVector(A=ScalarParameter() {defaultValue=1, parameterName=TileX}, B=Constant() {r=1}))
  n39 = TextureSample(UVs=NamedRerouteDeclaration(None=@n38) {name=Vine UV, nodeColor=<rgba 0 0.6953 1 1>}) {texture=/Game/Environments/Asteria/Props/GrindRailVines/GrindRailVines_3D_basecolor}
  NamedRerouteDeclaration(None=@n39.R) {name=BaseColorRedChannel, nodeColor=<rgba 0 0.9531 1 1>}
  n40 = TextureSample(UVs=NamedRerouteUsage()) {texture=/Game/Environments/Asteria/Sets/ZenGarden/Textures/T_Asteria_Zen_MossyGround_02}
  NamedRerouteDeclaration(None=@n40.R) {name=Diffuse_Moss, nodeColor=<rgba 0 1 0.6719 1>}
  n41 = TextureCoordinate()
  MaterialExpressionRemap_0 = MaterialXRemap!(
   None=OneMinus(None=ComponentMask(None=@n41) {g=true})
   InputLow
   InputHigh
   TargetLow=ScalarParameter() {bUseCustomPrimitiveData=true, parameterName=Unified UV Start}
   TargetHigh=ScalarParameter() {bUseCustomPrimitiveData=true, parameterName=Unified UV End, primitiveDataIndex=1}
  )
  NamedRerouteDeclaration(None=AppendVector(A=ComponentMask(None=@n41) {r=true}, B=@MaterialExpressionRemap_0)) {name=Unified UV, nodeColor=<rgba 0.4531 1 0 1>}
  NamedRerouteDeclaration(None=@n2.RGB) {name=Normaltexture, nodeColor=<rgba 0 0.1797 1 1>}
  n42 = TextureSample(UVs=NamedRerouteUsage()) {samplerSource=SSM_Wrap_WorldGroupSettings, texture=/ChronoEnviroItems/JungleGrindRails/Textures/T_JungleGrindRail_A_D}
  n43 = LinearInterpolate(A=@n39.RGB, B=Multiply(A=@n42.R, B=@n39.RGB), Alpha=@n39.RGB)
  n44 = Saturate(None=Add(A=NamedRerouteUsage()) {constB=0.7})
  n45 = Multiply(A=Desaturation(None=@n40.RGB, Fraction=Constant() {r=0.2}))
  NamedRerouteDeclaration(None=LinearInterpolate(A=StaticSwitchParameter(True=@n43, False=@n42.RGB) {parameterName=Is3xGrindrail}, B=Multiply(A=@n44, B=@n45), Alpha=NamedRerouteUsage())) {name=BaseColor, nodeColor=<rgba 0 0.0625 1 1>}
  n46 = Max(
   A=NamedRerouteUsage()
   B=Multiply(A=ScalarParameter() {bUseCustomPrimitiveData=true, parameterName=Glow switch, primitiveDataIndex=13}, B=NamedRerouteUsage())
  )
  n47 = ComponentMask(None=WorldPosition() {worldPositionShaderOffset=WPT_ExcludeAllShaderOffsets}.XYZ) {g=true, r=true}
  n48 = ScalarParameter() {defaultValue=100, parameterName=WorldNoise_Size2}
  n49 = Multiply(
   A=DotProduct(A=VertexNormalWS(), B=Constant3Vector() {constant=<rgba 0 0 1 0>})
   B=TextureSampleParameter2D(UVs=Divide(A=@n47, B=@n48)) {parameterName=WorldNoise2, samplerType=SAMPLERTYPE_Masks, texture=/Game/Global/Textures/Default/Noise/T_Metal_Noise}.R
  )
  n50 = VectorParameter() {group=EmissiveColor, parameterName=EmissiveColorLight, primitiveDataIndex=2, defaultValue=<rgba 0 2 1 1>}
  n51 = NamedRerouteUsage()
  n52 = Multiply(A=@n50.RGB, B=LinearInterpolate(Alpha=Power(Base=@n51)) {constA=3, constB=30})
  n53 = NamedRerouteUsage()
  n54 = Multiply(A=Multiply(A=Multiply(A=@n46, B=@n49), B=Multiply(A=@n52, B=@n53)), B=ScalarParameter() {defaultValue=5.5, parameterName=IdleSweepHigh})
  n55 = ScalarParameter() {defaultValue=1.2, parameterName=Global Morning Emissive}
  n56 = ScalarParameter() {defaultValue=2, parameterName=Global DayEmissive}
  n57 = ScalarParameter() {defaultValue=2, parameterName=Global EvEning Emissive}
  n58 = ScalarParameter() {defaultValue=1.4, parameterName=Global Night Emissive}
  n59 = ScalarParameter() {defaultValue=3.5, parameterName=Global Idle Emissive}
  n60 = ScalarParameter() {defaultValue=1.5, parameterName=Global Idle night Emissive}
  n61 = Constant() {r=40}
  n62 = Constant() {r=10}
  n63 = LinearInterpolate(
   A=Multiply(A=MaterialFunctionCall(Value_Morning=@n55, Value_Day=@n56, Value_Evening=@n57, Value_Night=@n58) {materialFunction=&Scalar_Time_of_Day}.Result)
   B=LinearInterpolate(
    A=MaterialFunctionCall(Value_Morning=@n59, Value_Day=@n59, Value_Evening=@n59, Value_Night=@n60) {materialFunction=&Scalar_Time_of_Day}.Result
    B=MaterialFunctionCall(Value_Morning=@n61, Value_Day=@n61, Value_Evening=@n61, Value_Night=@n62) {materialFunction=&Scalar_Time_of_Day}.Result
    Alpha=NamedRerouteUsage()
   )
   Alpha=ScalarParameter() {bUseCustomPrimitiveData=true, defaultValue=1, parameterName=IdleTail light switch, primitiveDataIndex=6}
  )
  n64 = NamedRerouteUsage()
  n65 = ScalarParameter() {bUseCustomPrimitiveData=true, parameterName=diastanc Glow, primitiveDataIndex=17}
  n66 = Multiply(
   A=LinearInterpolate(A=Multiply(A=@n54, B=@n63) {constB=2}, B=LinearInterpolate(B=@n64, Alpha=@n65), Alpha=NamedRerouteUsage()) {constB=0}
   B=NamedRerouteUsage()
  )
  n67 = EyeAdaptation()
  NamedRerouteDeclaration(None=ShadingPathSwitch(Default=Divide(A=@n66, B=@n67.EyeAdaptation), Mobile=@n66)) {name=Emissive, nodeColor=<rgba 0.1484 1 0 1>}
  NamedRerouteDeclaration(None=@n13) {name=Player SphereMask, nodeColor=<rgba 0.5 0 1 1>}
  NamedRerouteDeclaration(None=@n19.R) {name=WorldNoise, nodeColor=<rgba 1 0 0.9609 1>}
  n68 = NamedRerouteUsage()
  n69 = OneMinus(None=ScalarParameter() {defaultValue=0.3, parameterName=WorldNoise_Moss})
  n70 = Transform(None=NamedRerouteUsage())
  n71 = Saturate(None=LinearInterpolate(
    A=ScalarParameter() {defaultValue=0.64, parameterName=Moss_Contrast_Low}
    B=ScalarParameter() {defaultValue=0.9904, parameterName=Moss_Contrast_high}
    Alpha=ComponentMask(None=@n70) {b=true}
   ))
  n72 = TextureSample(UVs=Divide(A=NamedRerouteUsage(), B=ScalarParameter() {defaultValue=1, parameterName=Moss_Var_Size})) {texture=/Game/Environments/Asteria/Sets/ZenGarden/Textures/T_Asteria_Zen_MossyGround_01_H}
  n73 = LinearInterpolate(
   A=ScalarParameter() {defaultValue=0.7697, parameterName=Moss_Var_min}
   B=ScalarParameter() {defaultValue=4.1391, parameterName=Moss_Var_max}
   Alpha=@n72.R
  )
  n74 = Subtract(A=@n1.R) {constB=0.5}
  n75 = LinearInterpolate(
   A=TextureSample(UVs=@n7) {samplerType=SAMPLERTYPE_LinearColor, texture=/ChronoEnviroItems/JungleGrindRails/Textures/T_JungleGrindRail_A_H}.R
   Alpha=SphereMask(
    A=Abs(None=@n74)
    B=ScalarParameter() {defaultValue=0.5, parameterName=MossUVPosition}
    Radius=ScalarParameter() {defaultValue=0.65, parameterName=MossSpread}
   ) {hardnessPercent=0}
  )
  n76 = ScalarParameter() {defaultValue=1, parameterName=MossSpread_Power}
  n77 = Multiply(
   A=Saturate(None=Add(A=@n68, B=@n69))
   B=StaticSwitchParameter(True=Multiply(A=@n71, B=@n73), False=Power(Base=@n75, Exp=@n76) {constExponent=0.6}) {parameterName=Is3xGrindrail}
  )
  n78 = NamedRerouteDeclaration(None=Power(Base=@n72.R) {constExponent=1}) {name=Moss_Height, nodeColor=<rgba 0 1 0.7891 1>}
  n79 = ScalarParameter() {defaultValue=2, parameterName=MossSpread_Contrast}
  NamedRerouteDeclaration(None=Multiply(A=MaterialFunctionCall(Transition Phase=@n77, Height Texture=@n78, Contrast=@n79) {materialFunction=&HeightLerp}.Alpha)) {name=MossMask, nodeColor=<rgba 0.0312 0 1 1>}
  n80 = NamedRerouteUsage()
  n81 = SphereMask(
   A=ComponentMask(None=@n80) {g=true}
   B=Constant()
   Radius=ScalarParameter() {bUseCustomPrimitiveData=true, defaultValue=0.05, parameterName=Glow start Radius, primitiveDataIndex=9}
  ) {attenuationRadius=0.1, hardnessPercent=0}
  n82 = Multiply(
   A=ScalarParameter() {bUseCustomPrimitiveData=true, defaultValue=20, parameterName="Glow start  Intensity", primitiveDataIndex=11}
   B=Power(Base=ComponentMask(None=@n81) {g=true}) {constExponent=3}
  )
  n83 = SphereMask(
   A=ComponentMask(None=@n80) {g=true}
   B=Constant() {r=1}
   Radius=ScalarParameter() {bUseCustomPrimitiveData=true, defaultValue=0.05, parameterName=Glow End Radius, primitiveDataIndex=10}
  ) {attenuationRadius=0.1, hardnessPercent=0}
  n84 = Multiply(
   A=Power(Base=ComponentMask(None=@n83) {g=true}) {constExponent=3}
   B=ScalarParameter() {bUseCustomPrimitiveData=true, defaultValue=20, parameterName="Glow  end Intensity", primitiveDataIndex=12}
  ) {constB=10}
  NamedRerouteDeclaration(None=StaticSwitch(True=Max(A=@n82, B=@n84), False=Constant(), Value=StaticBoolParameter() {parameterName=Enable Glow start and end})) {name=Glow start and end, nodeColor=<rgba 0.4297 1 0 1>}
  NamedRerouteUsage()
  NamedRerouteUsage()
  NamedRerouteUsage()
  NamedRerouteUsage()
  NamedRerouteUsage()
  Reroute(None=@n38)
  Reroute(None=@n39.R)
  Reroute(None=@n2.RGB)
  StaticSwitchParameter(True=Constant3Vector(), False=NamedRerouteUsage()) {parameterName=Is3xGrindrail}
  StaticSwitchParameter(True=Divide(A=WorldPosition().XY, B=ScalarParameter() {defaultValue=193.2055, parameterName=Moss_WorldSpaceSize}), False=@n27) {parameterName=Is3xGrindrail}
  TextureCoordinate()
 ]
}
